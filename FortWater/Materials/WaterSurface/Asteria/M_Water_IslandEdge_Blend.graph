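# Masked Single Layer Water material for water at island edges, usable on spline and instanced meshes, with dithered vertex-alpha fade.
# Scattering and absorption optionally vary by world-space ocean noise, then pass through a biome blend pinned at alpha zero.
# A two-texture panner function supplies normals, while white foam blends in via HeightLerp from vertex blue, world noise and panning foam texture.
# ZOffset shifts vertices vertically; mobile gets simpler color and separate opacity.
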Material /FortWater/Materials/WaterSurface/Asteria/M_Water_IslandEdge_Blend
graph {
 properties {
  bUseMaterialAttributes=true
  bUsedWithInstancedStaticMeshes=true
  bUsedWithSplineMeshes=true
  blendMode=BLEND_Masked
  pixelDepthOffsetMode=PDOM_Legacy
  shadingModel=MSM_SingleLayerWater
  usedShadingModels=SingleLayerWater
 }
 nodes [
  LinearInterpolate()
  n1 = VectorParameter() {group=Water Shading Model, parameterName=Scattering, defaultValue=<rgba 0.1268 0.1336 0.15 0.5>}
  n2 = Multiply(A=@n1.RGB, B=Divide(A=@n1.A) {constB=100})
  n3 = VectorParameter() {group=Water Shading Model, parameterName=Scattering_Noise, defaultValue=<rgba 0.1268 0.1336 0.15 0.5>}
  n4 = Multiply(A=@n3.RGB, B=Divide(A=@n3.A) {constB=100})
  n5 = WorldPosition()
  n6 = TextureSample(UVs=Divide(
    A=ComponentMask(None=@n5.XYZ) {g=true, r=true}
    B=ScalarParameter() {defaultValue=130000, group=Water Shading Model, parameterName=OceanColorNoise_Size}
   )) {samplerType=SAMPLERTYPE_Masks, texture=/FortWater/Textures/OceanNoise__Mask}
  n7 = StaticSwitchParameter(True=ShadingPathSwitch(Default=LinearInterpolate(A=@n2, B=@n4, Alpha=@n6.R), Mobile=@n2) {desc=Migrated from FeatureLevelSwitch}, False=@n2) {parameterName=UseOceanColorNoise}
  n8 = Add(B=@n7) {desc=ADD Foam Mask Here}
  n9 = QualitySwitch(Default=@n8, Low=@n7, High=@n8, Medium=@n7)
  n10 = VectorParameter() {group=Water Shading Model, parameterName=BiomeScattering, defaultValue=<rgba 0.2301 0.2396 0.1547 1>}
  n11 = VectorParameter() {group=Water Shading Model, parameterName=Absorption, defaultValue=<rgba 1 13.35 20 100>}
  n12 = Multiply(A=@n11.RGB, B=@n11.RGB)
  n13 = VectorParameter() {group=Water Shading Model, parameterName=Absorption_Noise, defaultValue=<rgba 1 13.35 20 100>}
  n14 = LinearInterpolate(A=@n12, B=Multiply(A=@n13.RGB, B=@n13.RGB), Alpha=@n6.R)
  n15 = VectorParameter() {group=Water Shading Model, parameterName=BiomeAbsorption, defaultValue=<rgba 1 0.642 0.4152 100>}
  n16 = MaterialFunctionCall!(
   Scattering A=ShadingPathSwitch(Default=@n9, Deferred=@n9, Forward=@n7, Mobile=@n7)
   Scattering B=Multiply(A=@n10.RGB, B=Divide(A=@n10.A) {constB=100})
   Absorption A=StaticSwitchParameter(True=ShadingPathSwitch(Default=@n14, Mobile=@n12) {desc=Migrated from FeatureLevelSwitch}, False=@n12) {parameterName=UseOceanColorNoise}
   Absorption B=Multiply(A=@n15.RGB, B=@n15.RGB)
   Alpha=Constant()
  ) {materialFunction=/FortWater/Materials/Functions/BlendWaterShadingValues}
  Multiply(B=@n16.Scattering)
  n17 = VertexColor()
  OneMinus(None=@n17.R)
  n18 = ScalarParameter() {group=Water Shading Model, parameterName=Surface Opacity}
  Reroute(None=LinearInterpolate(A=@n18, B=ScalarParameter() {defaultValue=1.1, group=Water Shading Model, parameterName=Surface Opacity2}, Alpha=@n17.R))
  n19 = AppendVector(A=Constant2Vector(), B=ScalarParameter() {parameterName=ZOffset})
  n20 = SetMaterialAttributes!(
   MaterialAttributes
   Base Color=ShadingPathSwitch(
    Default=VectorParameter() {parameterName=BaseColor, defaultValue=<rgba 0 0.1 0.1 0.1>}.RGB
    Mobile=VectorParameter() {parameterName=BaseColorMobile, defaultValue=<rgba 0.0039 0.008 0 0.1>}.RGB
   ) {desc=Migrated from FeatureLevelSwitch}
   Specular=ScalarParameter() {defaultValue=0.225, group=Water Shading Model, parameterName=Specular}
   Roughness=ScalarParameter() {defaultValue=0.05, desc=below 0.5 DLSS has sparkly bbloom highlights, group=Water Shading Model, parameterName=Roughness}
   Opacity=ShadingPathSwitch(Default=@n18, Mobile=ScalarParameter() {defaultValue=1, group=Water Shading Model, parameterName=Surface Opacity_MOBILE}) {desc=Migrated from FeatureLevelSwitch}
   Normal=MaterialFunctionCall!(Normal, WPO, Far Mask Override, Additional Mask, LowDetailBiome_Mask) {materialFunction=/FortWater/Materials/Functions/WaterTwoTexturePanner}.Normal
   World Position Offset=ShadingPathSwitch(Default=Constant3Vector(), Mobile=@n19) {desc=Migrated from FeatureLevelSwitch}
   Opacity Mask=MaterialFunctionCall(Alpha Threshold=VertexColor().A) {materialFunction=&DitherTemporalAA}.Result
  ) {attributeSetTypes [
    #1
    #2
    #3
    #4
    #5
    #6
    #7
   ]}
  n21 = Constant3Vector() {constant=<rgba 1 1 1 0>}
  n22 = ScalarParameter() {defaultValue=0.4877, parameterName=NoiseContrast_1}
  n23 = ScalarParameter() {defaultValue=5.5392, parameterName=NoiseContrast_2}
  n24 = WorldPosition() {worldPositionShaderOffset=WPT_ExcludeAllShaderOffsets}
  n25 = TextureSample(UVs=Divide(A=ComponentMask(None=@n24.XYZ) {g=true, r=true}, B=ScalarParameter() {defaultValue=79682.875, parameterName=WorldNoise_Size})) {texture=/Game/Global/Textures/Default/Noise/T_Tiling_Noise_01}
  n26 = Multiply(
   A=Saturate(None=LinearInterpolate(A=@n22, B=@n23, Alpha=@n25.R))
   B=Multiply(A=@n17.B, B=ScalarParameter() {defaultValue=0.248, parameterName=Foam})
  )
  n27 = TextureCoordinate() {coordinateIndex=1}
  n28 = AppendVector(A=ScalarParameter() {defaultValue=3, parameterName=Foam_X}, B=ScalarParameter() {defaultValue=3, parameterName=Foam_y})
  n29 = ScalarParameter() {parameterName=Foam_Speed_x}
  n30 = ScalarParameter() {defaultValue=-0.04, parameterName=Foam_Speed_y}
  n31 = TextureSample(UVs=Panner(Coordinate=Multiply(A=@n27, B=@n28), Speed=AppendVector(A=@n29, B=@n30))) {texture=/FortWater/Textures/T_Foam_DistanceField_Fade_02}
  n32 = ScalarParameter() {defaultValue=0.3346, parameterName=FoamContrast}
  SetMaterialAttributes!(
   MaterialAttributes=BlendMaterialAttributes(
    A=@n20
    B=SetMaterialAttributes!(MaterialAttributes, Base Color=@n21, Specular) {attributeSetTypes [#1, #2]}
    Alpha=MaterialFunctionCall(Transition Phase=@n26, Height Texture=@n31.R, Contrast=@n32) {materialFunction=&HeightLerp}.Alpha
   )
   World Position Offset=@n19
  ) {attributeSetTypes [#6]}
  SingleLayerWaterMaterialOutput!(
   ScatteringCoefficients=@n16.Scattering
   AbsorptionCoefficients=Divide(B=@n16.Absorption)
   PhaseG=CollectionParameter() {collection=/Water/Materials/MPC/MPC_Water, parameterName=Ansotropy}
   ColorScaleBehindWater
  )
 ]
}
